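# Unlit skeletal-mesh material for the clothing on the WashyFlare vehicle wheel cosmetic.
# Emissive color blends from near-white to the ClothingColorMain parameter, masked by the summed green channels of the rim mask and its LOD texture.
# SimpleGrassWind drives world position offset for cloth motion, with intensity, weight and speed parameters.
# The attributes then pass through the shared vehicle damage function, then the de-rez transition function.
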Material /VehicleCosmetics/Wheels/WashyFlare/Materials/M_WhasyFlare_Clothing
graph {
 properties {
  bAutomaticallySetUsageInEditor=false
  bUseMaterialAttributes=true
  bUsedWithSkeletalMesh=true
  shadingModel=MSM_Unlit
  usedShadingModels=Unlit
 }
 nodes [
  n1 = LinearInterpolate(
   A=Constant4Vector() {constant=<rgba 0.9801 0.9727 1 1>}.RGB
   B=VectorParameter() {parameterName=ClothingColorMain, defaultValue=<rgba 1 0 0.0135 1>}.RGB
   Alpha=Add(
    A=TextureSample() {samplerType=SAMPLERTYPE_LinearColor, texture=/VehicleCosmetics/Wheels/WashyFlare/Textures/T_WashyFlare_Rim_M}.G
    B=TextureSample() {samplerType=SAMPLERTYPE_LinearColor, texture=/VehicleCosmetics/Wheels/WashyFlare/Textures/T_WashyFlare_Rim_M_LOD}.G
   )
  )
  n2 = MaterialFunctionCall!(
   WindIntensity=ScalarParameter() {defaultValue=0.5, parameterName=Intensity}
   WindWeight=ScalarParameter() {defaultValue=0.25, parameterName=Weight}
   WindSpeed=ScalarParameter() {defaultValue=0.4, parameterName=Speed}
   AdditionalWPO=Constant() {r=1}
  ) {materialFunction=/Engine/Functions/Engine_MaterialFunctions01/WorldPositionOffset/SimpleGrassWind}
  MaterialFunctionCall(In=MaterialFunctionCall(Material=SetMaterialAttributes!(MaterialAttributes, Emissive Color=@n1, World Position Offset=@n2.Result) {attributeSetTypes [#1, #2]}) {materialFunction=&MF_BaseVehicle_Damage}.Result) {materialFunction=&MF_VehicleDeRez_Transition}
 ]
}
