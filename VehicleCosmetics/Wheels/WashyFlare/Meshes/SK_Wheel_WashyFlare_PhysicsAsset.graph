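# Physics asset for the WashyFlare vehicle cosmetic wheel skeletal mesh, giving the wheel and its spinner simple collision bodies.
# Each of the wheel_jnt and spinner_jnt bones has one capsule rotated onto its side, with the spinner's smaller than the wheel's.
# A single constraint joins spinner to wheel, with matching swing and twist limit angles.
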PhysicsAsset /VehicleCosmetics/Wheels/WashyFlare/Meshes/SK_Wheel_WashyFlare_PhysicsAsset
bodies {
 spinner_jnt [{shapeType=Capsule, length=36.8942, radius=36.5782, center=<0.3127 -9.0458 0.0461>, rotation=<rot 0.1671 -90 -89.3938>}]
 wheel_jnt [{shapeType=Capsule, length=51.2016, radius=51.2016, center=<0 -9.7719 0>, rotation=<rot -0.0007 -90 0>}]
}
constraints [{bone1Name=spinner_jnt, bone2Name=wheel_jnt, swing1LimitDegrees=45, swing2LimitDegrees=45, twistLimitDegrees=45}]
